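# Masked subsurface material for balloon particles in the ball explosion effect, usable on Niagara mesh particles and sprites.
# Plasma noise, mapped by local position with a per-particle random offset, feeds a floored dissolve mask driven by the Dissolve dynamic parameter.
# A second soft noise, scaled by the WPO Noise dynamic parameter, creates displacement along vertex normals.
# Particle color supplies scaled diffuse and an adjustable emissive term through MF_FX_EmissiveAdjustments.
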Material /SunnyLomaBallGameplay/Assets/FX/BallExplosion/M_BallExplosion_Balloons
graph {
 properties {
  bUsedWithNiagaraMeshParticles=true
  bUsedWithNiagaraSprites=true
  blendMode=BLEND_Masked
  pixelDepthOffsetMode=PDOM_Legacy
  previewMesh=&/SM_Balloon
  shadingModel=MSM_Subsurface
  usedShadingModels=Subsurface
 }
 nodes [
  n1 = ComponentMask(None=LocalPosition().XYZ) {b=true, g=true}
  n2 = ParticleRandom()
  n3 = TextureSampleParameter2D(UVs=Multiply(A=Add(A=@n1, B=@n2)) {constB=0.001}) {
   parameterName=BaseTexture
   samplerType=SAMPLERTYPE_LinearGrayscale
   texture=/Game/Effects/Fort_Effects/Emotes/UltraLight/Texture/T_Plasma_C_Greyscale
  }
  n4 = Power(Base=@n3.R) {constExponent=1.5}
  n5 = OneMinus(None=DynamicParameter() {defaultValue=<rgba 1 0 1 1>, paramNames [WPO Noise, Dissolve, Param3, Param4]}.Dissolve)
  Floor(None=Add(A=Subtract(A=@n4), B=Multiply(A=@n5) {constB=2}))
  n6 = ParticleColor()
  MaterialFunctionCall(Emissive=LinearInterpolate(A=Constant(), B=@n6.RGB, Alpha=ScalarParameter() {parameterName=Add Emissive})) {materialFunction=&MF_FX_EmissiveAdjustments}
  Multiply(A=@n6.RGB, B=ScalarParameter() {defaultValue=0.3})
  n7 = LocalPosition()
  n8 = Add(A=Multiply(A=ComponentMask(None=@n7.XYZ) {b=true, g=true}) {constB=0.002}, B=ParticleRandom())
  Multiply(
   A=DynamicParameter() {defaultValue=<rgba 1 0 1 1>, paramNames [WPO Noise, Dissolve, Param3, Param4]}.WPO Noise
   B=Multiply(
    A=TextureSampleParameter2D(UVs=@n8) {parameterName=BaseTexture02, texture=/Game/Effects/Fort_Effects/Emotes/UltraLight/Texture/T_SoftNoise_BC7}.B
    B=VertexNormalWS()
   )
  )
  Reroute(None=Multiply(A=@n6.RGB, B=ScalarParameter() {defaultValue=0.3, desc=Diffuse Contribution, parameterName=Diffuse Mult, sliderMax=1}))
  ScalarParameter() {defaultValue=0.2, parameterName=Param_1}
  VectorParameter() {parameterName=MSR, defaultValue=<rgba 0 0.7 0.384 1>}
 ]
}
