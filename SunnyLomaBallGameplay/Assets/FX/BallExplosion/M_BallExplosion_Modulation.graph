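# Unlit Modulate-blend Niagara sprite material for the ball explosion effect, tinting the scene behind a soft round particle.
# It lerps from neutral white to particle color scaled by a Modulation Value dynamic parameter (index 1).
# The lerp weight combines particle alpha, Modulation Alpha and a radial sphere mask shaped by exponential density.
# A feature level switch adds depth fade to that mask, except on ES2 and ES3_1.
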Material /SunnyLomaBallGameplay/Assets/FX/BallExplosion/M_BallExplosion_Modulation
graph {
 properties {
  bAutomaticallySetUsageInEditor=false
  bUsedWithNiagaraSprites=true
  blendMode=BLEND_Modulate
  shadingModel=MSM_Unlit
  usedShadingModels=Unlit
 }
 nodes [
  n1 = ParticleColor()
  n2 = DynamicParameter() {parameterIndex=1, paramNames [Modulation Value, Modulation Alpha, Param3, Param4]}
  n3 = TextureCoordinate()
  n4 = Constant() {r=0.5}
  n5 = MaterialFunctionCall!(
   Depth=OneMinus(None=SphereMask(A=@n3, B=@n4, Radius=@n4) {hardnessPercent=0})
   Density=ScalarParameter() {defaultValue=2, parameterName=Density}
   Use Exp2
  ) {materialFunction=/Engine/Functions/Engine_MaterialFunctions01/Density/ExponentialDensity}
  n6 = DepthFade(Opacity=@n5.Result, FadeDistance=ScalarParameter() {defaultValue=400, parameterName=FadeDistance})
  LinearInterpolate(
   B=Multiply(A=@n1.RGB, B=@n2.Modulation Value)
   Alpha=Multiply(
    A=Multiply(A=@n1.A, B=@n2.Modulation Alpha)
    B=FeatureLevelSwitch(Default=@n6, ES2=@n5.Result, ES3_1=@n5.Result, SM4_REMOVED=@n6, SM5=@n6, SM6=@n6)
   )
  ) {constA=1}
 ]
}
