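# Translucent, two-sided material for Niagara mesh particles in the ball explosion effect, shaping spline meshes into glowing streaks.
# World position offset math combines SplineThicken, with Width Base driving both base and tip, and a Gravity-scaled Z offset weighted by squared vertex red.
# Hue-shifted particle color is scaled for emissive.
# A wave ribbon mask is cut by a Progression wipe along U and a Perlin noise dissolve, all particle-driven through dynamic parameters.
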Material /SunnyLomaBallGameplay/Assets/FX/BallExplosion/M_BallExplosion_SplineThicken
graph {
 properties {
  bAutomaticallySetUsageInEditor=false
  bUsedWithNiagaraMeshParticles=true
  blendMode=BLEND_Translucent
  previewMesh=&/SM_SplineMesh_Default
  twoSided=true
  usedShadingModels=Default Lit
 }
 nodes [
  n1 = DynamicParameter() {parameterIndex=1, paramNames [Progression, Width Base, Width Tip, Gravity]}
  n2 = VertexColor()
  n3 = Multiply(A=@n2.R, B=@n2.R)
  n4 = Constant3Vector() {constant=<rgba 0 0 1 0>}
  Add(
   A=MaterialFunctionCall(WidthBase=@n1.Width Base, WidthTip=@n1.Width Base, Expand U or V UV Channel=StaticBool()) {materialFunction=&SplineThicken}.WorldPositionOffset
   B=Multiply(A=Multiply(A=@n3, B=@n4), B=DynamicParameter() {parameterIndex=1, paramNames [Progression, Width Base, Width Tip, Gravity]}.Gravity)
  )
  n5 = DynamicParameter() {defaultValue=<rgba 1 1 0 1>, paramNames [Hue Shift, Emissive, Dissolve, Param4]}
  n6 = ParticleColor()
  MaterialFunctionCall(Emissive=Multiply(
    A=MaterialFunctionCall(Hue Shift Percentage=@n5.Hue Shift, Texture=@n6.RGB) {materialFunction=&HueShift}.Result
    B=@n5.Emissive
   )) {materialFunction=&MF_FX_EmissiveAdjustments}
  n7 = Multiply(A=TextureCoordinate(), B=AppendVector(A=ScalarParameter() {defaultValue=8}, B=ScalarParameter() {defaultValue=1, parameterName=Param_1}))
  n8 = ComponentMask(None=TextureCoordinate()) {r=true}
  n9 = OneMinus(None=@n8)
  n10 = DynamicParameter() {parameterIndex=1, paramNames [Progression, Width Base, Width Tip, Gravity]}
  n11 = Floor(None=Add(A=Subtract(A=@n9), B=Multiply(A=@n10.Progression) {constB=2}))
  n12 = ScalarParameter() {defaultValue=0.2, parameterName=Param_2}
  n13 = Subtract(A=TextureSample(UVs=Multiply(A=@n7, B=@n12)) {samplerType=SAMPLERTYPE_LinearGrayscale, texture=&T_SimplePerlin_Greyscale}.R)
  n14 = Multiply(A=OneMinus(None=DynamicParameter() {defaultValue=<rgba 1 1 0 1>, paramNames [Hue Shift, Emissive, Dissolve, Param4]}.Dissolve)) {constB=2}
  Multiply(
   A=Multiply(
    A=TextureSample(UVs=@n7) {samplerType=SAMPLERTYPE_Grayscale, texture=&/T_Wave_Ribbon}.R
    B=Multiply(A=@n8, B=@n11)
   )
   B=Floor(None=Add(A=@n13, B=@n14))
  )
  Power(Base=@n8) {constExponent=0.5}
 ]
}
