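# Niagara celebratory ball explosion effect that bursts balloons, confetti and smoke, all tinted from a shared system color curve.
# Mesh balloons fly out, then rise under upward acceleration and aerodynamic drag, with late-life curves driving material parameters.
# Flipbook confetti flutters down through curl noise, as a dense GPU emitter and a sparser Confetti_Low copy.
# Smoke uses spawn-group masking for two puff variants; glow-sphere and spline-mesh emitters show no renderers.
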NiagaraSystem /SunnyLomaBallGameplay/Assets/FX/BallExplosion/NS_BallExplosion
dependencies {
 usedDataInterfaces [
  &NiagaraDataInterfaceCurve
  &NiagaraDataInterfaceSpriteRendererInfo
  &NiagaraDataInterfaceColorCurve
 ]
 usedDynamicInputs []
 usedModules [
  &self:SystemSpawnScript.SystemScriptSource.SystemScriptGraph.NiagaraNodeAssignment_0.SetVariables_#1
  &SystemState
 ]
}
emitters {
 Balloons {
  emitterUpdateScript [
   &EmitterState(Life Cycle Mode=::NewEnumerator1, Loop Behavior=::NewEnumerator1, Loop Delay=0.1, Loop Duration=1, UseLoopDelay=true)
   &SpawnBurst_Instantaneous(Spawn Count=80, Use Spawn Probability=false, UseLoopCountLimit=false)
  ]
  particleSpawnScript [
   &InitializeParticle(
    Color=&ColorFromCurve(ColorCurve=$System.ColorCurve, CurveIndex=&RandomRangeFloat(Minimum=0, Maximum=1))
    Lifetime=&FloatFromCurve(
     FloatCurve=DataInterface() {propertyValues {Curve {keys [
         {arriveTangent=5.666, interpMode=RCIM_Cubic, leaveTangent=5.666, tangentMode=RCTM_User, time=0, value=1}
         {arriveTangent=12.0119, interpMode=RCIM_Cubic, leaveTangent=12.0119, tangentMode=RCTM_User, time=1, value=5}
        ]}}}
     CurveIndex=&RandomRangeFloat(Minimum=0, Maximum=1)
    )
    Mesh Scale Mode=::NewEnumerator1
    Mesh Uniform Scale Max=1.4
    UsePositionOffset=false
    Write Material Random=true
   )
   &ShapeLocation(MaskSpawnGroup=false, Override Seed=false, Sphere Radius=50, UseOwnerScale=true)
   &AddVelocity(
    Module.Use Falloff Distance=false
    Override Seed=false
    Velocity Mode=::NewEnumerator1
    Velocity Speed=&RandomRangeFloat(Minimum=1375.1477, Maximum=4777.9521)
   )
   &InitialMeshOrientation(Mesh Orientation Mode=::NewEnumerator2, Use Rotation Vector=false)
  ]
  particleUpdateScript [
   &ParticleState()
   &AccelerationForce(Acceleration=&RandomRangeVector(Minimum=<0 0 396.8>, Maximum=<0 0 1316.7043>))
   &AerodynamicDrag(
    Aerodynamic Drag:NiagaraFloat=&FloatFromCurve(FloatCurve=DataInterface() {propertyValues {Curve {keys [{arriveTangent=-0.4155, interpMode=RCIM_Cubic, leaveTangent=-0.4155, tangentMode=RCTM_User, time=0, value=1}, <at 0.5 0.1>]}}})
    Aerodynamic Rotational Drag=0.3
    Lift Contribution=0.6
    Rotation Strength=0.4
    Use Linear Drag=true
    Use Rotational Drag=true
   )
   &SolveForcesAndVelocity(Clamp Velocity=false, Limit Acceleration=false)
   &ScaleMeshSize(Scale Factor=&VectorFromFloat(Value=&FloatFromCurve(FloatCurve=DataInterface() {propertyValues {Curve {
         defaultValue=1
         keys [
          <at 0 0>
          {arriveTangent=7.612, leaveTangent=7.612, time=0.0399, value=1.0111}
          {arriveTangent=-3.2064, interpMode=RCIM_Cubic, leaveTangent=-3.2064, time=0.0581, value=0.4655}
          {arriveTangent=0.7302, interpMode=RCIM_Cubic, leaveTangent=0.7302, time=0.1117, value=0.7809}
          <at 0.79 1>
          <at 0.8 0.1227>
          {tangentMode=RCTM_User, time=0.81, value=1}
          <at 1 1>
          {tangentMode=RCTM_User, time=1, value=1}
         ]
        }}})))
   &DynamicMaterialParameters(
    Index 0 Param 1=&FloatFromCurve(
     FloatCurve=DataInterface() {propertyValues {Curve {
        defaultValue=1
        keys [
         {arriveTangent=8.892, interpMode=RCIM_Cubic, leaveTangent=8.892, tangentMode=RCTM_User, time=0.8, value=0}
         {tangentMode=RCTM_User, time=1, value=1}
        ]
       }}}
     Scale Curve=250
    )
    Index 0 Param 2=&FloatFromCurve(FloatCurve=DataInterface() {propertyValues {Curve {
        defaultValue=1
        keys [
         {arriveTangent=4.0548, interpMode=RCIM_Cubic, leaveTangent=4.0548, tangentMode=RCTM_User, time=0.8, value=0}
         {tangentMode=RCTM_User, time=1, value=1}
        ]
       }}})
    Module.Index0Float0Write=true
    Module.Index0Float1Write=true
    Module.Index0Float2Write=true
    Module.Index0Float3Write=true
   )
   &AlignSpriteToMeshOrientation(SetAlignment=true, SetFacing=true)
  ]
  rendererClasses [&NiagaraMeshRendererProperties]
  renderers [{rendererClass=&NiagaraMeshRendererProperties, rendererIndex=0}]
 }
 Confetti_GPU {
  simTarget=GPUComputeSim
  emitterUpdateScript [
   &EmitterState(Life Cycle Mode=::NewEnumerator1, Loop Behavior=::NewEnumerator1, Loop Delay=0.1, Loop Duration=1, UseLoopDelay=true)
   &SpawnBurst_Instantaneous(Spawn Count=500, Use Spawn Probability=false, UseLoopCountLimit=false)
  ]
  particleSpawnScript [
   &InitializeParticle(
    Color=&ColorFromCurve(ColorCurve=$System.ColorCurve, CurveIndex=&RandomRangeFloat(Minimum=0, Maximum=1))
    Lifetime=&FloatFromCurve(
     FloatCurve=DataInterface() {propertyValues {Curve {keys [<at 0 2 RCIM_Cubic>, {arriveTangent=8.5901, interpMode=RCIM_Cubic, leaveTangent=8.5901, tangentMode=RCTM_User, time=1, value=5}]}}}
     CurveIndex=&RandomRangeFloat(Minimum=0, Maximum=1)
    )
    Sprite Size Mode=::NewEnumerator1
    Uniform Sprite Size Max=60
    Uniform Sprite Size Min=30
    UsePositionOffset=false
    Write Material Random=true
   )
   &ShapeLocation(MaskSpawnGroup=false, Override Seed=false, Sphere Radius=50, UseOwnerScale=true)
   &AddVelocity(
    Module.Use Falloff Distance=false
    Override Seed=false
    Velocity Mode=::NewEnumerator1
    Velocity Speed=&Lerp_Float(
     A=2495.0972
     B=6960.1616
     Alpha=&FloatFromCurve(
      FloatCurve=DataInterface() {propertyValues {Curve {
         defaultValue=1
         keys [<at 0 0 RCIM_Cubic>, {arriveTangent=1.9042, interpMode=RCIM_Cubic, leaveTangent=1.9042, tangentMode=RCTM_User, time=1, value=1}]
        }}}
      CurveIndex=&RandomRangeFloat(Minimum=0, Maximum=1)
     )
    )
   )
  ]
  particleUpdateScript [
   &ParticleState()
   &ScaleSpriteSize(Uniform Curve Sprite Scale=DataInterface() {propertyValues {Curve {
       defaultValue=0
       keys [
        {arriveTangent=4.1785, interpMode=RCIM_Cubic, leaveTangent=4.1785, tangentMode=RCTM_User, time=0, value=0}
        {arriveTangent=1.1179, interpMode=RCIM_Cubic, leaveTangent=1.1179, tangentMode=RCTM_User, time=0.0623, value=1.3}
        {arriveTangent=0.9937, interpMode=RCIM_Cubic, leaveTangent=0.9937, tangentMode=RCTM_User, time=0.1, value=0.925}
        {arriveTangent=-0.0395, interpMode=RCIM_Cubic, leaveTangent=-0.0395, tangentMode=RCTM_User, time=0.5961, value=0.95}
        {arriveTangent=0.0224, interpMode=RCIM_Cubic, leaveTangent=0.0224, tangentMode=RCTM_User, time=1, value=0}
       ]
      }}})
   &Drag(Drag=&RandomRangeFloat(Minimum=4, Maximum=6), Use Linear Drag=true, Use Rotational Drag=false)
   &GravityForce(Gravity=&RandomRangeVector(Minimum=<0 0 -326>, Maximum=<0 0 -560.6946>))
   &CurlNoiseForce(Bias Noise=false, Noise Frequency=10, Noise Strength=670.6714, Pan Noise=true, Pan Noise Field=<0 0 -2>)
   &SolveForcesAndVelocity(Clamp Velocity=false, Limit Acceleration=false)
   &SubUVAnimation(
    Playback Scale Curve=DataInterface() {propertyValues {Curve {keys [
        {arriveTangent=-0.932, interpMode=RCIM_Cubic, leaveTangent=-0.932, tangentMode=RCTM_User, time=0.25, value=1.2}
        {interpMode=RCIM_Cubic, leaveTangent=-1.2656, time=1, value=1}
       ]}}}
    "Sprite Renderer "=DataInterface() {propertyValues {SpriteRenderer=&self:Confetti_GPU_0.NiagaraSpriteRendererProperties_0}}
    Start Time Offset:NiagaraFloat=&RandomRangeFloat(Minimum=0, Maximum=1)
    SubUV Animation Mode:ENiagara_SubUVLookupModeV2=::NewEnumerator3
    UseEndFrame=false
    UsePlayRateScale=true
    UseStartFrame=false
   )
   &DynamicMaterialParameters(Module.Index0Float0Write=true, Module.Index0Float1Write=true, Module.Index0Float2Write=true, Module.Index0Float3Write=true)
   &ScaleColor(
    Linear Color Curve=DataInterface() {propertyValues {
      AlphaCurve {keys [<at 0 1>,<at 0.7978 1>,<at 1 0>]}
      BlueCurve {keys [<at 0 0.6137>,<at 0.0935 1>,<at 1 1>]}
      GreenCurve {keys [<at 0 2.3848>,<at 0.0935 1>,<at 1 1>]}
      RedCurve {keys [<at 0 5>,<at 0.0935 1>,<at 1 1>]}
     }}
    Scale Mode=::NewEnumerator2
   )
  ]
  rendererClasses [&NiagaraSpriteRendererProperties]
  renderers [{rendererClass=&NiagaraSpriteRendererProperties}]
 }
 Confetti_Low {
  emitterUpdateScript [
   &EmitterState(Life Cycle Mode=::NewEnumerator1, Loop Behavior=::NewEnumerator1, Loop Delay=0.1, Loop Duration=1, UseLoopDelay=true)
   &SpawnBurst_Instantaneous(Spawn Count=60, Use Spawn Probability=false, UseLoopCountLimit=false)
  ]
  particleSpawnScript [
   &InitializeParticle(
    Color=&ColorFromCurve(ColorCurve=$System.ColorCurve, CurveIndex=&RandomRangeFloat(Minimum=0, Maximum=1))
    Lifetime=&FloatFromCurve(
     FloatCurve=DataInterface() {propertyValues {Curve {keys [<at 0 2 RCIM_Cubic>, {arriveTangent=8.5901, interpMode=RCIM_Cubic, leaveTangent=8.5901, tangentMode=RCTM_User, time=1, value=5}]}}}
     CurveIndex=&RandomRangeFloat(Minimum=0, Maximum=1)
    )
    Sprite Size Mode=::NewEnumerator1
    Uniform Sprite Size Max=70
    Uniform Sprite Size Min=50
    UsePositionOffset=false
    Write Material Random=true
   )
   &ShapeLocation(MaskSpawnGroup=false, Override Seed=false, Sphere Radius=50, UseOwnerScale=true)
   &AddVelocity(
    Module.Use Falloff Distance=false
    Override Seed=false
    Velocity Mode=::NewEnumerator1
    Velocity Speed=&Lerp_Float(
     A=2495.0972
     B=6960.1616
     Alpha=&FloatFromCurve(
      FloatCurve=DataInterface() {propertyValues {Curve {
         defaultValue=1
         keys [<at 0 0 RCIM_Cubic>, {arriveTangent=1.9042, interpMode=RCIM_Cubic, leaveTangent=1.9042, tangentMode=RCTM_User, time=1, value=1}]
        }}}
      CurveIndex=&RandomRangeFloat(Minimum=0, Maximum=1)
     )
    )
   )
  ]
  particleUpdateScript [
   &ParticleState()
   &ScaleSpriteSize(Uniform Curve Sprite Scale=DataInterface() {propertyValues {Curve {
       defaultValue=0
       keys [
        {arriveTangent=4.1785, interpMode=RCIM_Cubic, leaveTangent=4.1785, tangentMode=RCTM_User, time=0, value=0}
        {arriveTangent=1.1179, interpMode=RCIM_Cubic, leaveTangent=1.1179, tangentMode=RCTM_User, time=0.0623, value=1.3}
        {arriveTangent=0.9937, interpMode=RCIM_Cubic, leaveTangent=0.9937, tangentMode=RCTM_User, time=0.1, value=0.925}
        {arriveTangent=-0.0395, interpMode=RCIM_Cubic, leaveTangent=-0.0395, tangentMode=RCTM_User, time=0.5961, value=0.95}
        {arriveTangent=0.0224, interpMode=RCIM_Cubic, leaveTangent=0.0224, tangentMode=RCTM_User, time=1, value=0}
       ]
      }}})
   &Drag(Drag=&RandomRangeFloat(Minimum=4, Maximum=6), Use Linear Drag=true, Use Rotational Drag=false)
   &GravityForce(Gravity=&RandomRangeVector(Minimum=<0 0 -326>, Maximum=<0 0 -560.6946>))
   &CurlNoiseForce(Bias Noise=false, Noise Frequency=10, Noise Strength=670.6714, Pan Noise=true, Pan Noise Field=<0 0 -2>)
   &SolveForcesAndVelocity(Clamp Velocity=false, Limit Acceleration=false)
   &SubUVAnimation(
    Playback Scale Curve=DataInterface() {propertyValues {Curve {keys [
        {arriveTangent=-0.932, interpMode=RCIM_Cubic, leaveTangent=-0.932, tangentMode=RCTM_User, time=0.25, value=1.2}
        {interpMode=RCIM_Cubic, leaveTangent=-1.2656, time=1, value=1}
       ]}}}
    "Sprite Renderer "=DataInterface() {propertyValues {SpriteRenderer=&self:Confetti_Low_0.NiagaraSpriteRendererProperties_0}}
    Start Time Offset:NiagaraFloat=&RandomRangeFloat(Minimum=0, Maximum=1)
    SubUV Animation Mode:ENiagara_SubUVLookupModeV2=::NewEnumerator3
    UseEndFrame=false
    UsePlayRateScale=true
    UseStartFrame=false
   )
   &DynamicMaterialParameters(Module.Index0Float0Write=true, Module.Index0Float1Write=true, Module.Index0Float2Write=true, Module.Index0Float3Write=true)
   &ScaleColor(
    Linear Color Curve=DataInterface() {propertyValues {
      AlphaCurve {keys [<at 0 1>,<at 0.7978 1>,<at 1 0>]}
      BlueCurve {keys [<at 0 0.6137>,<at 0.0935 1>,<at 1 1>]}
      GreenCurve {keys [<at 0 2.3848>,<at 0.0935 1>,<at 1 1>]}
      RedCurve {keys [<at 0 5>,<at 0.0935 1>,<at 1 1>]}
     }}
    Scale Mode=::NewEnumerator2
   )
  ]
  rendererClasses [&NiagaraSpriteRendererProperties]
  renderers [{rendererClass=&NiagaraSpriteRendererProperties}]
 }
 LW_GlowSphereMesh2 {rendererClasses [], renderers []}
 LW_GlowSphereMesh2001 {rendererClasses [], renderers []}
 Smoke {
  emitterUpdateScript [
   &EmitterState(Life Cycle Mode=::NewEnumerator1, Loop Behavior=::NewEnumerator1, Loop Delay=0.15, Loop Duration=1, UseLoopDelay=true)
   &SpawnBurst_Instantaneous(Spawn Count=4, Use Spawn Probability=false, UseLoopCountLimit=false)
   &SpawnBurst_Instantaneous(Spawn Count=8, Spawn Group=1, Use Spawn Probability=false, UseLoopCountLimit=false) {module=SpawnBurst_Instantaneous001}
  ]
  particleSpawnScript [
   &InitializeParticle(
    Color=&ColorFromCurve(
     ColorCurve=$System.ColorCurve
     CurveIndex=&MaskFloatBySpawnGroup(
      Masked Float=&ReturnNormalizedExecIndex(Normalized Index Scale=0.4786, Normalized Index Includes One=false)
      Passthrough Float=&ReturnNormalizedExecIndex()
     )
    )
    Lifetime=&MaskFloatBySpawnGroup(
     Masked Float=&RandomRangeFloat(Minimum=1.4, Maximum=2)
     Passthrough Float=&RandomRangeFloat(Minimum=3, Maximum=5)
    )
    Sprite Rotation Mode=::NewEnumerator2
    Sprite Size Mode=::NewEnumerator3
    Uniform Sprite Size=&MaskFloatBySpawnGroup(
     Masked Float=&RandomRangeFloat(Minimum=600, Maximum=1000)
     Passthrough Float=&RandomRangeFloat(Minimum=1200, Maximum=1600)
    )
    UsePositionOffset=false
    Write Material Random=true
   )
   &ShapeLocation(
    MaskSpawnGroup=false
    Rotation Angle:NiagaraFloat=&RandomRangeFloat(Minimum=0, Maximum=1)
    Rotation Axis=&RandomVector(Vector Scale=1)
    Rotation Mode=::NewEnumerator1
    Sphere Distribution=::NewEnumerator3
    UseOwnerScale=true
   )
   &AddVelocity(
    Module.Use Falloff Distance=false
    Override Seed=false
    Velocity Mode=::NewEnumerator1
    Velocity Speed=&MaskFloatBySpawnGroup(
     Masked Float=&Lerp_Float(A=0, B=800, Alpha=&ReturnNormalizedExecIndex(Normalized Index Includes One=false))
     Passthrough Float=&RandomRangeFloat(Minimum=1340.5298, Maximum=2400)
    )
   )
  ]
  particleUpdateScript [
   &ParticleState()
   &ScaleSpriteSize(Uniform Curve Sprite Scale=DataInterface() {propertyValues {Curve {
       defaultValue=0
       keys [
        {arriveTangent=4.1785, interpMode=RCIM_Cubic, leaveTangent=4.1785, tangentMode=RCTM_User, time=0, value=0.3}
        {arriveTangent=1.1179, interpMode=RCIM_Cubic, leaveTangent=1.1179, tangentMode=RCTM_User, time=0.2042, value=0.7505}
        {arriveTangent=0.0224, interpMode=RCIM_Cubic, leaveTangent=0.0224, tangentMode=RCTM_User, time=1, value=1}
       ]
      }}})
   &DynamicMaterialParameters(
    Index 0 Param 1=&Multiply_Float(A=$Particles.Age, B=0.2)
    Index 0 Param 3=&FloatFromCurve(
     FloatCurve=DataInterface() {propertyValues {Curve {
        defaultValue=0
        keys [
         <at 0 1 RCIM_Cubic>
         {arriveTangent=-0.4988, interpMode=RCIM_Cubic, leaveTangent=-0.4988, tangentMode=RCTM_User, time=0.227, value=0.1}
         {tangentMode=RCTM_User, time=1, value=0}
        ]
       }}}
     CurveIndex=$Particles.Age
    )
    Index 2 Param 1=1
    Index 2 Param 2=&RandomRangeFloat(Minimum=0, Maximum=1)
    Index 2 Param 3=&FloatFromCurve(
     FloatCurve=DataInterface() {propertyValues {Curve {defaultValue=1, keys [<at 0 0 RCIM_Cubic>, {tangentMode=RCTM_User, time=1, value=1}]}}}
     Scale Curve=3
    )
    Index 2 Param 4=1
    Module.Index0Float0Write=true
    Module.Index0Float1Write=true
    Module.Index0Float2Write=true
    Module.Index0Float3Write=true
    Module.Index1Float0Write=true
    Module.Index1Float1Write=true
    Module.Index1Float2Write=true
    Module.Index1Float3Write=true
    Module.Index2Float0Write=true
    Module.Index2Float1Write=true
    Module.Index2Float2Write=true
    Module.Index2Float3Write=true
    Write Parameter Index 1=::NewEnumerator1
    Write Parameter Index 2=::NewEnumerator1
   )
   &ScaleColor(
    Scale Alpha=&FloatFromCurve(
     FloatCurve=DataInterface() {propertyValues {Curve {keys [
         <at 0 0 RCIM_Cubic>
         {arriveTangent=-0.1413, interpMode=RCIM_Cubic, leaveTangent=-0.1413, tangentMode=RCTM_User, time=0.1, value=1}
         {arriveTangent=-0.1874, interpMode=RCIM_Cubic, leaveTangent=-0.1874, tangentMode=RCTM_User, time=0.5, value=1}
         <at 1 0>
        ]}}}
     Scale Curve=&MaskFloatBySpawnGroup(Masked Float=3, Passthrough Float=0.5)
    )
    Scale RGB=&VectorFromFloat(Value=0.6)
    ScaleA=true
    ScaleRGB=true
   )
   &Drag(Drag=4, Use Linear Drag=true, Use Rotational Drag=false)
   &SolveForcesAndVelocity(Clamp Velocity=false, Limit Acceleration=false)
  ]
  rendererClasses [&NiagaraSpriteRendererProperties]
  renderers [{rendererClass=&NiagaraSpriteRendererProperties}]
 }
 SplineMesh {rendererClasses [], renderers []}
}
summary {
 systemName=NS_BallExplosion
 emitters [
  {emitterName=LW_GlowSphereMesh2, rendererClasses []}
  {emitterName=Confetti_GPU, simTarget=GPUComputeSim, rendererClasses [&NiagaraSpriteRendererProperties]}
  {emitterName=SplineMesh, rendererClasses []}
  {emitterName=Balloons, rendererClasses [&NiagaraMeshRendererProperties]}
  {emitterName=LW_GlowSphereMesh2001, rendererClasses []}
  {emitterName=Smoke, rendererClasses [&NiagaraSpriteRendererProperties]}
  {emitterName=Confetti_Low, rendererClasses [&NiagaraSpriteRendererProperties]}
 ]
 userVariables []
}
